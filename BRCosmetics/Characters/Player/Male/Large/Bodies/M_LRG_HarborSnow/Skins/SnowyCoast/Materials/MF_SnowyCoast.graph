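# Material function for the SnowyCoast skin of the HarborSnow player body, adding a timed, reactive color overlay to incoming material attributes.
# Time since GameTimeFromBP, divided by a duration, samples a curve atlas that drives a reveal through the FX mask's red channel, limited by green.
# The reveal blends in a noise-mixed two-color tint with faint emissive, altered metallic, specular and roughness, and a flattened normal.
# Debugtime and a looping time node are unconnected.
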MaterialFunction /BRCosmetics/Characters/Player/Male/Large/Bodies/M_LRG_HarborSnow/Skins/SnowyCoast/Materials/MF_SnowyCoast
graph {nodes [
  n1 = Time()
  Frac(None=Multiply(A=@n1) {constB=0.1})
  n2 = GetMaterialAttributes(None=FunctionInput() {inputName=Material, inputType=FunctionInput_MaterialAttributes}) {attributeGetTypes [
    #1
    #2
    #3
    #4
    #5
    #6
   ]}
  n3 = TextureSample(UVs=Multiply(A=TextureCoordinate(), B=ScalarParameter() {defaultValue=3, parameterName=Tiling})) {texture=/Game/Global/Textures/Default/Noise/3xPerlin_M}
  n4 = LinearInterpolate(
   A=VectorParameter() {parameterName=Color01, defaultValue=<rgba 0.9375 0.4748 0.0897 1>}.RGB
   B=VectorParameter() {parameterName=Color02, defaultValue=<rgba 1 0.1958 0 1>}.RGB
   Alpha=SmoothStep(Value=@n3.R) {constMax=0.8}
  )
  n5 = TextureSample() {
   samplerType=SAMPLERTYPE_Masks
   texture=/BRCosmetics/Characters/Player/Male/Large/Bodies/M_LRG_HarborSnow/Skins/SnowyCoast/Textures/T_SnowyCoast_Body_FX1
  }
  n6 = Subtract(A=@n1, B=ScalarParameter() {parameterName=GameTimeFromBP})
  n7 = ScalarParameter() {defaultValue=10, parameterName=SnowyCoast_Duration}
  n8 = CurveAtlasRowParameter(CurveTime=Saturate(None=Divide(A=@n6, B=@n7))) {
   atlas=/BRCosmetics/Characters/Player/Male/Large/Bodies/M_LRG_HarborSnow/Skins/SnowyCoast/Textures/CA_SnowyCoast
   curve=/BRCosmetics/Characters/Player/Male/Large/Bodies/M_LRG_HarborSnow/Skins/SnowyCoast/Textures/C_SnowyCoast
   parameterName=SnowyCoastReactivity
  }
  n9 = Subtract(A=Multiply(A=@n8.R) {constB=2}, B=OneMinus(None=@n5.R))
  n10 = Multiply(A=@n5.G, B=LinearInterpolate(A=Saturate(None=@n9), B=@n8.R, Alpha=@n8.R))
  n11 = SetMaterialAttributes!(
   MaterialAttributes=@n2.MaterialAttributes
   BaseColor=LinearInterpolate(A=@n2.BaseColor, B=@n4, Alpha=@n10)
   Metallic=LinearInterpolate(A=@n2.Metallic, Alpha=@n10)
   Specular=LinearInterpolate(A=@n2.Specular, Alpha=@n10) {constB=0.15}
   Roughness=LinearInterpolate(A=@n2.Roughness, Alpha=@n10) {constB=0.3}
   EmissiveColor=LinearInterpolate(A=@n2.EmissiveColor, B=Multiply(A=@n4) {constB=0.2}, Alpha=@n10)
   Normal=LinearInterpolate(A=@n2.Normal, B=Constant3Vector() {constant=<rgba 0 0 1 0>}, Alpha=@n10)
  ) {attributeSetTypes [
    #1
    #2
    #3
    #4
    #5
    #6
   ]}
  FunctionOutput(None=@n11)
  NamedRerouteDeclaration(None=@n3.R) {nodeColor=<rgba 0.2422 1 0 1>}
  ScalarParameter() {defaultValue=0.5, parameterName=Debugtime}
 ]}
