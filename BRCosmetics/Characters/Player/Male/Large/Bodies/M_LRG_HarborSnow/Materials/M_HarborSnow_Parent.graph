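# Parent character material for the HarborSnow body, layering skin-specific effects between the shared CharacterParent PreFX and PostFX functions.
# Tints base color and emissive, using the color parameter's alpha to select a gradient hue-shifted variant.
# An optional battery decal, placed front and back by UV offsets, animates a timed emissive charge and fade.
# Static switches add blink-driven eye overrides split by pre-skinned position, or swap everything for the SnowyCoast function.
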Material /BRCosmetics/Characters/Player/Male/Large/Bodies/M_LRG_HarborSnow/Materials/M_HarborSnow_Parent
graph {
 properties {
  bUseMaterialAttributes=true
  bUsedWithMorphTargets=true
  bUsedWithSkeletalMesh=true
  pixelDepthOffsetMode=PDOM_Legacy
  shadingModel=MSM_FromMaterialExpression
  usedShadingModels=Default Lit | Subsurface | Clear Coat | Subsurface Profile | Cloth
 }
 nodes [
  n1 = MaterialFunctionCall() {materialFunction=&MF_CharacterParent_2023_PreFX}
  MaterialFunctionCall(Material=NamedRerouteUsage(), M=@n1.M, Pre Skinned Local Position=@n1.M, Distance Blend=@n1.M, Fresnel=@n1.M) {materialFunction=&MF_CharacterParent_2023_PostFX}
  n2 = VertexInterpolator(VS=ComponentMask(None=LocalPosition() {localOrigin=InstancePreSkinning}.XYZ) {r=true})
  NamedRerouteDeclaration(None=Saturate(None=Ceil(None=@n2.PS))) {name=EyeMask_Left, nodeColor=<rgba 0 0.4141 1 1>}
  NamedRerouteDeclaration(None=@n1.Tech Art Mask) {name=TAMask, nodeColor=<rgba 0.7344 1 0 1>}
  NamedRerouteDeclaration(None=@n1.Result) {name=InMA, nodeColor=<rgba 0.4766 1 0 1>}
  n3 = NamedRerouteUsage()
  n4 = GetMaterialAttributes(None=NamedRerouteUsage()) {attributeGetTypes [#1, #2]}
  n5 = VectorParameter() {group=01_HarborSnow_EyeColor, parameterName=Character_BaseColor, defaultValue=<rgba 1 1 1 1>}
  n6 = ComponentMask(None=NamedRerouteUsage()) {b=true}
  n7 = Saturate(None=Divide(A=Add(A=@n6) {constB=-8}) {constB=150})
  n8 = Multiply(A=@n7, B=@n7)
  n9 = Constant() {r=-1}
  n10 = Constant() {r=1}
  n11 = Multiply(
   A=@n8
   B=Multiply(A=ScalarParameter() {defaultValue=0.1, parameterName=Pink_HueShiftAmount}, B=If(A=@n5.A, "A > B"=@n9, "A < B"=@n10) {constB=2.01})
  )
  n12 = If(A=@n5.A, "A > B"=Constant() {r=2}, "A < B"=Constant() {r=1}) {constB=2.01}
  n13 = LinearInterpolate(
   A=@n5.RGB
   B=Desaturation(
    None=MaterialFunctionCall(Hue Shift Percentage=@n11, Texture=@n5.RGB) {materialFunction=&HueShift}.Result
    Fraction=ScalarParameter() {defaultValue=-0.5, parameterName=Pink_DesaturationAmount}
   )
   Alpha=Saturate(None=Subtract(A=@n5.A, B=@n12))
  ) {constAlpha=1}
  n14 = NamedRerouteUsage()
  n15 = SetMaterialAttributes!(
   MaterialAttributes=@n4.MaterialAttributes
   Base Color=LinearInterpolate(A=@n4.Base Color, B=Multiply(A=@n4.Base Color, B=@n13), Alpha=ComponentMask(None=@n14) {r=true})
   Emissive Color=Multiply(A=@n13, B=@n4.Emissive Color)
  ) {attributeSetTypes [#1, #2]}
  n16 = GetMaterialAttributes(None=@n15) {attributeGetTypes [#2, #1, #3]}
  n17 = VectorParameter() {group=01_HarborSnow_Reactivity, parameterName=Battery_BaseColor, defaultValue=<rgba 0.1746 0.0595 0.008 1>}
  n18 = LinearInterpolate(A=@n17.RGB, B=Constant3Vector() {constant=<rgba 0.1441 0.0972 0.0383 1>}, Alpha=Saturate(None=Subtract(A=@n17.A)))
  n19 = Subtract(A=Time(), B=ScalarParameter() {defaultValue=-99999, group=01_HarborSnow_Reactivity, parameterName=Reactivity_TL})
  n20 = Add(A=Constant() {r=4}) {constB=0}
  n21 = StaticSwitchParameter(
   True=ScalarParameter() {group=01_HarborSnow_Reactivity, parameterName=DebugTimeline, sliderMax=1}
   False=Saturate(None=Divide(A=@n19, B=@n20))
  ) {group=01_HarborSnow_Reactivity, parameterName=UseDebugTimeline, sortPriority=0}
  n22 = Saturate(None=Multiply(A=OneMinus(None=@n21)) {constB=4.5}) {desc=FadeOut}
  n23 = VectorParameter() {
   group=01_HarborSnow_Reactivity
   parameterName=BatteryPositionBack
   channelNames {a="", b=Scale, g=Y, r=X}
   defaultValue=<rgba 0.292 0.051 0.05 0>
  }
  n24 = MaterialFunctionCall(UVs=Add(A=TextureCoordinate(), B=ComponentMask(None=@n23) {g=true, r=true}), Texture Scale=@n23.Scale) {materialFunction=&ScaleUVsByCenter}
  n25 = TextureSample(UVs=@n24.UVs) {
   samplerSource=SSM_Clamp_WorldGroupSettings
   samplerType=SAMPLERTYPE_Masks
   texture=/BRCosmetics/Characters/Player/Male/Large/Bodies/M_LRG_HarborSnow/Textures/T_HarborSnow_Body_SL1
  }
  n26 = VectorParameter() {
   group=01_HarborSnow_Reactivity
   parameterName=BatteryPositionFront
   channelNames {a="", b=Scale, g=Y, r=X}
   defaultValue=<rgba 0.2985 0.338 0.105 0>
  }
  n27 = MaterialFunctionCall(UVs=Add(A=TextureCoordinate(), B=ComponentMask(None=@n26) {g=true, r=true}), Texture Scale=@n26.Scale) {materialFunction=&ScaleUVsByCenter}
  n28 = TextureSample(UVs=@n27.UVs) {
   samplerSource=SSM_Clamp_WorldGroupSettings
   samplerType=SAMPLERTYPE_Masks
   texture=&T_HarborSnow_Body_SL1
  }
  n29 = Max(A=Min(A=@n25.B, B=@n24.0-1 mask), B=Min(A=@n28.B, B=@n27.0-1 mask))
  MaterialExpressionMax_3 = Max(A=Min(A=@n25.RGB, B=@n24.0-1 mask), B=Min(A=@n28.RGB, B=@n27.0-1 mask))
  n30 = Convert(None=@MaterialExpressionMax_3) {
   convertInputs [{
     type=Vector2
     expressionInput {
      expression=&self:MaterialExpressionMax_3
      inputName=None
      mask=0
      maskA=0
      maskB=0
      maskG=0
      maskR=0
      outputIndex=0
     }
    }]
   convertOutputs [{type=Scalar}, {type=Scalar}]
  }
  n31 = Multiply(A=Clamp(None=Multiply(A=@n21) {constB=1.25}) {desc=Charge, maxDefault=0.999}) {constB=5}
  n32 = Constant() {r=5}
  n33 = Min(A=Multiply(A=ComponentMask(None=@n29) {r=true}, B=@n30), B=Subtract(A=Modulo!(A=@n31, B=@n32), B=Multiply(A=@n30) {constB=4}))
  n34 = Add(A=@n16.Emissive Color, B=Multiply(A=Multiply(A=@n18) {constB=5}, B=Multiply(A=@n22, B=@n33)))
  n35 = Saturate(None=@n30)
  n36 = LinearInterpolate(A=@n16.Base Color, B=@n17.RGB, Alpha=@n35)
  n37 = LinearInterpolate(
   A=@n16.Roughness
   B=ScalarParameter() {defaultValue=0.78, group=01_HarborSnow_Reactivity, parameterName=Battery_Roughness, sliderMax=1}
   Alpha=@n35
  )
  n38 = StaticSwitchParameter(
   True=BlendMaterialAttributes(
    A=@n16.MaterialAttributes
    B=SetMaterialAttributes!(MaterialAttributes=@n16.MaterialAttributes, Emissive Color=@n34, Base Color=@n36, Roughness=@n37) {attributeSetTypes [#2, #1, #3]}
    Alpha=ScalarParameter() {defaultValue=1, parameterName=BatteryOpacity}
   )
   False=@n15
  ) {group=01_HarborSnow_Reactivity, parameterName=HarborSnow_BatteryReactivity, sortPriority=0}
  n39 = GetMaterialAttributes(None=@n38) {attributeGetTypes [#1, #4, #3]}
  n40 = NamedRerouteUsage()
  n41 = ScalarParameter() {group=01_HarborSnow_EyeColor, parameterName=L_Blink_pose, sliderMax=1}
  n42 = OneMinus(None=@n40)
  n43 = ScalarParameter() {group=01_HarborSnow_EyeColor, parameterName=R_Blink_pose, sliderMax=1}
  n44 = Min(A=Max(A=Min(A=@n40, B=@n41), B=Min(A=@n42, B=@n43)), B=ComponentMask(None=NamedRerouteUsage()) {g=true})
  n45 = SetMaterialAttributes!(
   MaterialAttributes=@n39.MaterialAttributes
   Base Color=LinearInterpolate(A=@n39.Base Color, B=VectorParameter() {group=01_HarborSnow_EyeColor, parameterName=Eye_BaseColor}.RGB, Alpha=@n44)
   Specular=LinearInterpolate(A=@n39.Specular, B=ScalarParameter() {defaultValue=0.5, group=01_HarborSnow_EyeColor, parameterName=Eye_Specular}, Alpha=@n44)
   Roughness=LinearInterpolate(A=@n39.Roughness, B=ScalarParameter() {defaultValue=0.5, group=01_HarborSnow_EyeColor, parameterName=Eye_Roughness}, Alpha=@n44)
  ) {attributeSetTypes [#1, #4, #3]}
  NamedRerouteDeclaration(None=StaticSwitchParameter(
    True=MaterialFunctionCall!(Material=@n3) {materialFunction=/BRCosmetics/Characters/Player/Male/Large/Bodies/M_LRG_HarborSnow/Skins/SnowyCoast/Materials/MF_SnowyCoast}.Result
    False=StaticSwitchParameter(True=@n45, False=@n38) {group=01_HarborSnow_Reactivity, parameterName=HarborSnow_EyeColorOverride, sortPriority=0}
   ) {group=01_HarborSnow_Reactivity, parameterName=HarborSnow_UseSnowyCoast, sortPriority=0}) {name=OutMA, nodeColor=<rgba 1 0 0.7969 1>}
  NamedRerouteDeclaration(None=@n1.M) {name=Masks, nodeColor=<rgba 1 0 0.9609 1>}
  NamedRerouteDeclaration(None=@n1.Fresnel) {name=InFresnel, nodeColor=<rgba 1 0.4688 0 1>}
  NamedRerouteDeclaration(None=@n1.Pre skinned Local Position) {name=PreskinnedPos, nodeColor=<rgba 0 1 0.9531 1>}
  OneMinus(None=@n8)
  ScalarParameter() {defaultValue=1, group=01_HarborSnow_Reactivity, parameterName=ReactivityFadeOut, sliderMax=1}
  ScalarParameter() {defaultValue=0.99, group=01_HarborSnow_Reactivity, parameterName=ReactivityCharge, sliderMax=1}
  Sign(None=@n5.A)
 ]
}
