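# Control Rig for the M_LRG_HarborSnow player body, adding procedural corrective deformation and gear attachment.
# A five-step sequence runs the shared DeformRig function, weighted parent constraints for dyn_ gear master bones, and head/neck_twist constraints for head deform roots.
# It then spreads head twist into neck bones via swing-twist slerp, and spherical pose readers (debug draw on) drive head correctives.
# Older gear constraint chains and jaw/chin nodes are left unconnected.
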ControlRigBlueprint /BRCosmetics/Characters/Player/Male/Large/Bodies/M_LRG_HarborSnow/Meshes/M_LRG_HarborSnow_CtrlRig
graphs [
 RigVMFunctionLibrary {links [], nodes []}
 RigVMFunction_Sequence_1_ContainedGraph {nodes [
   n1 = RigVMFunctionEntryNode()
   RigVMFunction_Sequence = &RigVMFunction_Sequence(ExecuteContext=@n1.ExecuteContext)
   n2 = &RigVMFunction_Sequence(ExecuteContext=@RigVMFunction_Sequence.B)
   n3 = &RigVMFunction_Sequence(ExecuteContext=@n2.B)
   n4 = &RigVMFunction_Sequence(ExecuteContext=@n3.B)
   RigVMFunctionReturnNode(A=@RigVMFunction_Sequence.A, B=@n2.A, C=@n3.A, D=@n4.A, E=@n4.B)
  ]}
 RigVMModel {nodes [
   &RigUnit_ParentConstraint(
    AdvancedSettings=(RotationOrderForFilter=XZY)
    Child=(Type=Bone,Name="dyn_box_r_lrg_MSTR")
    ExecutePin=()
    Parents="((Item=(Type=Bone,Name=\"deform_elbow_out_r\"),Weight=0.0001),(Item=(Type=Bone,Name=\"pelvis\"),Weight=0.0002),(Item=(Type=Bone,Name=\"deform_lat_r\"),Weight=0.0005),(Item=(Type=Bone,Name=\"spine_02\"),Weight=0.0021),(Item=(Type=Bone,Name=\"upperarm_r\"),Weight=0.00 ...(truncated)"
    Weight=1
   )
   > &RigUnit_ParentConstraint(
    AdvancedSettings=(RotationOrderForFilter=XZY)
    Child=(Type=Bone,Name="dyn_box_l_MSTR")
    Parents="((Item=(Type=Bone,Name=\"deform_upperarm_fr_l\"),Weight=0.0001),(Item=(Type=Bone,Name=\"deform_lat_l\"),Weight=0.0003),(Item=(Type=Bone,Name=\"deform_elbow_out_l\"),Weight=0.001),(Item=(Type=Bone,Name=\"spine_02\"),Weight=0.0024),(Item=(Type=Bone,Name=\"spine_03\"), ...(truncated)"
    Weight=1
   )
   > &RigUnit_ParentConstraint(
    AdvancedSettings=(RotationOrderForFilter=XZY)
    Child=(Type=Bone,Name="dyn_pouch_l_MSTR")
    Parents="((Item=(Type=Bone,Name=\"deform_lat_l\"),Weight=0.0002),(Item=(Type=Bone,Name=\"neck_01\"),Weight=0.0003),(Item=(Type=Bone,Name=\"pelvis\"),Weight=0.0004),(Item=(Type=Bone,Name=\"neck_02\"),Weight=0.0004),(Item=(Type=Bone,Name=\"head\"),Weight=0.0009),(Item=(Type=Bo ...(truncated)"
    Weight=1
   )
   > &RigUnit_ParentConstraint(
    AdvancedSettings=(RotationOrderForFilter=XZY)
    Child=(Type=Bone,Name="dyn_buckle_l_MSTR")
    Parents="((Item=(Type=Bone,Name=\"deform_upperarm_bk_l\"),Weight=0.0004),(Item=(Type=Bone,Name=\"deform_bicep_l\"),Weight=0.002),(Item=(Type=Bone,Name=\"neck_01\"),Weight=0.0035),(Item=(Type=Bone,Name=\"spine_02\"),Weight=0.0045),(Item=(Type=Bone,Name=\"neck_02\"),Weight=0.0 ...(truncated)"
    Weight=1
   )
   > &RigUnit_ParentConstraint(
    AdvancedSettings=(RotationOrderForFilter=XZY)
    Child=(Type=Bone,Name="dyn_buckle_r_MSTR")
    Parents="((Item=(Type=Bone,Name=\"deform_upperarm_bk_r\"),Weight=0.0004),(Item=(Type=Bone,Name=\"deform_bicep_r\"),Weight=0.002),(Item=(Type=Bone,Name=\"neck_01\"),Weight=0.0035),(Item=(Type=Bone,Name=\"spine_02\"),Weight=0.0044),(Item=(Type=Bone,Name=\"neck_02\"),Weight=0.0 ...(truncated)"
    Weight=1
   )
   > &RigUnit_ParentConstraint(
    AdvancedSettings=(RotationOrderForFilter=XZY)
    Child=(Type=Bone,Name="dyn_wire_MSTR")
    Parents="((Item=(Type=Bone,Name=\"pelvis\"),Weight=0.0003),(Item=(Type=Bone,Name=\"deform_elbow_out_r\"),Weight=0.0008),(Item=(Type=Bone,Name=\"deform_lat_r\"),Weight=0.0009),(Item=(Type=Bone,Name=\"spine_02\"),Weight=0.0053),(Item=(Type=Bone,Name=\"spine_03\"),Weight=0.0072 ...(truncated)"
    Weight=1
   )
   > &RigUnit_ParentConstraint(
    AdvancedSettings=(RotationOrderForFilter=XZY)
    Child=(Type=Bone,Name="dyn_box_r_small_MSTR")
    Parents="((Item=(Type=Bone,Name=\"deform_upperarm_fr_r\"),Weight=0.0001),(Item=(Type=Bone,Name=\"deform_lat_r\"),Weight=0.0003),(Item=(Type=Bone,Name=\"deform_elbow_out_r\"),Weight=0.0013),(Item=(Type=Bone,Name=\"spine_02\"),Weight=0.0022),(Item=(Type=Bone,Name=\"spine_03\") ...(truncated)"
    Weight=1
   )
   &RigUnit_ParentConstraint(
    AdvancedSettings=(RotationOrderForFilter=XZY)
    Child=(Type=Bone,Name="dyn_box_r_lrg_MSTR")
    ExecutePin=()
    Parents="((Item=(Type=Bone,Name=\"deform_upperarm_up_r\"),Weight=0.0065),(Item=(Type=Bone,Name=\"upperarm_twist_02_r\"),Weight=0.0072),(Item=(Type=Bone,Name=\"spine_03\"),Weight=0.0073),(Item=(Type=Bone,Name=\"deform_upperarm_bk_r\"),Weight=0.0073),(Item=(Type=Bone,Name=\"u ...(truncated)"
    Weight=1
   )
   > &RigUnit_ParentConstraint(
    AdvancedSettings=(RotationOrderForFilter=XZY)
    Child=(Type=Bone,Name="dyn_box_l_MSTR")
    Parents="((Item=(Type=Bone,Name=\"deform_upperarm_bk_l\"),Weight=0.0149),(Item=(Type=Bone,Name=\"upperarm_l\"),Weight=0.0184),(Item=(Type=Bone,Name=\"upperarm_twist_02_l\"),Weight=0.0215),(Item=(Type=Bone,Name=\"deform_upperarm_up_l\"),Weight=0.0234),(Item=(Type=Bone,Name= ...(truncated)"
    Weight=1
   )
   > &RigUnit_ParentConstraint(
    AdvancedSettings=(RotationOrderForFilter=XZY)
    Child=(Type=Bone,Name="dyn_pouch_l_MSTR")
    Parents="((Item=(Type=Bone,Name=\"deform_upperarm_up_l\"),Weight=0.0246),(Item=(Type=Bone,Name=\"spine_04\"),Weight=0.0298),(Item=(Type=Bone,Name=\"upperarm_twist_02_l\"),Weight=0.0734),(Item=(Type=Bone,Name=\"deform_upperarm_fr_l\"),Weight=0.0796),(Item=(Type=Bone,Name=\"s ...(truncated)"
    Weight=1
   )
   > &RigUnit_ParentConstraint(
    AdvancedSettings=(RotationOrderForFilter=XZY)
    Child=(Type=Bone,Name="dyn_buckle_l_MSTR")
    Parents="((Item=(Type=Bone,Name=\"head\"),Weight=0.0215),(Item=(Type=Bone,Name=\"deform_upperarm_fr_l\"),Weight=0.0253),(Item=(Type=Bone,Name=\"upperarm_twist_02_l\"),Weight=0.0287),(Item=(Type=Bone,Name=\"deform_upperarm_up_l\"),Weight=0.0288),(Item=(Type=Bone,Name=\"spine ...(truncated)"
    Weight=1
   )
   > &RigUnit_ParentConstraint(
    AdvancedSettings=(RotationOrderForFilter=XZY)
    Child=(Type=Bone,Name="dyn_buckle_r_MSTR")
    Parents="((Item=(Type=Bone,Name=\"head\"),Weight=0.0216),(Item=(Type=Bone,Name=\"deform_upperarm_fr_r\"),Weight=0.0251),(Item=(Type=Bone,Name=\"upperarm_twist_02_r\"),Weight=0.0285),(Item=(Type=Bone,Name=\"deform_upperarm_up_r\"),Weight=0.0286),(Item=(Type=Bone,Name=\"spine ...(truncated)"
    Weight=1
   )
   > &RigUnit_ParentConstraint(
    AdvancedSettings=(RotationOrderForFilter=XZY)
    Child=(Type=Bone,Name="dyn_box_r_small_MSTR")
    Parents="((Item=(Type=Bone,Name=\"deform_upperarm_bk_r\"),Weight=0.0147),(Item=(Type=Bone,Name=\"upperarm_r\"),Weight=0.0161),(Item=(Type=Bone,Name=\"upperarm_twist_02_r\"),Weight=0.0192),(Item=(Type=Bone,Name=\"deform_upperarm_up_r\"),Weight=0.0223),(Item=(Type=Bone,Name= ...(truncated)"
    Weight=1
   )
   > &RigUnit_ParentConstraint(
    AdvancedSettings=(RotationOrderForFilter=XZY)
    Child=(Type=Bone,Name="dyn_wire_MSTR")
    Parents="((Item=(Type=Bone,Name=\"deform_upperarm_up_r\"),Weight=0.0139),(Item=(Type=Bone,Name=\"deform_upperarm_bk_r\"),Weight=0.016),(Item=(Type=Bone,Name=\"upperarm_r\"),Weight=0.0166),(Item=(Type=Bone,Name=\"upperarm_twist_02_r\"),Weight=0.017),(Item=(Type=Bone,Name=\"c ...(truncated)"
    Weight=1
   )
   RigUnit_BeginExecution = &RigUnit_BeginExecution()
   RigVMFunction_Sequence = RigVMAggregateNode(ExecuteContext=@RigUnit_BeginExecution.ExecutePin)
   FN_DeformRig_FNC = RigVMFunctionReferenceNode(
    DeformRig_ Curve="(deformRig_l_upperarm_up_alpha_72_#1=False,deformRig_r_upperarm_up_alpha_73_#2=False,deformRig_l_upperarm_fr_alpha_74_#3=False,deformRig_r_upperarm_fr_alpha_75_CB51B8 ...(truncated)"
    DeformRig_Bool="(upperarm_1_#4=True,elbow_3_#5=True,wrist_7_#6=True,groin_8_#7=True,glute_17_#8=True,knee_9_40C408324CA261 ...(truncated)"
    DeformRig_Debug=(upperarm_debug_18_#4=False,wrist_debug_20_#6=False,lat_debug_26_#9=False,groinGlute_debug_27_#7=False)
    DeformRig_Value="(upperarm_up_18_#10=(Rotation=(X=0,Y=0,Z=0,W=1),Translation=(X=0,Y=0,Z=8),Scale3D=(X=0.7,Y=1,Z=1)),upperarm_fwd_fr_19_#11=(Rotation=(X=0,Y=0,Z=0,W=1),Translation=(X=0,Y=10,Z=0),Scale3D=(X=0.5,Y=1,Z= ...(truncated)"
    ExecuteContext=@RigVMFunction_Sequence.A
   )
   &RigUnit_ParentConstraint(
    AdvancedSettings=(RotationOrderForFilter=XZY)
    Child=(Type=Bone,Name="dyn_box_r_lrg_MSTR")
    Parents="((Item=(Type=Bone,Name=\"deform_upperarm_up_r\"),Weight=0.0065),(Item=(Type=Bone,Name=\"upperarm_twist_02_r\"),Weight=0.0072),(Item=(Type=Bone,Name=\"spine_03\"),Weight=0.0073),(Item=(Type=Bone,Name=\"deform_upperarm_bk_r\"),Weight=0.0073),(Item=(Type=Bone,Name=\"u ...(truncated)"
    Weight=1
    ExecutePin=@RigVMFunction_Sequence.B
   )
   > &RigUnit_ParentConstraint(
    AdvancedSettings=(RotationOrderForFilter=XZY)
    Child=(Type=Bone,Name="dyn_box_l_MSTR")
    Parents="((Item=(Type=Bone,Name=\"deform_upperarm_bk_l\"),Weight=0.0149),(Item=(Type=Bone,Name=\"upperarm_l\"),Weight=0.0185),(Item=(Type=Bone,Name=\"upperarm_twist_02_l\"),Weight=0.0215),(Item=(Type=Bone,Name=\"deform_upperarm_up_l\"),Weight=0.0234),(Item=(Type=Bone,Name= ...(truncated)"
    Weight=1
   )
   > &RigUnit_ParentConstraint(
    AdvancedSettings=(RotationOrderForFilter=XZY)
    Child=(Type=Bone,Name="dyn_pouch_l_MSTR")
    Parents="((Item=(Type=Bone,Name=\"deform_upperarm_up_l\"),Weight=0.0192),(Item=(Type=Bone,Name=\"upperarm_twist_02_l\"),Weight=0.0579),(Item=(Type=Bone,Name=\"deform_upperarm_fr_l\"),Weight=0.0626),(Item=(Type=Bone,Name=\"spine_05\"),Weight=0.1284),(Item=(Type=Bone,Name=\"c ...(truncated)"
    Weight=1
   )
   > &RigUnit_ParentConstraint(
    AdvancedSettings=(RotationOrderForFilter=XZY)
    Child=(Type=Bone,Name="dyn_buckle_l_MSTR")
    Parents="((Item=(Type=Bone,Name=\"head\"),Weight=0.0215),(Item=(Type=Bone,Name=\"deform_upperarm_fr_l\"),Weight=0.0253),(Item=(Type=Bone,Name=\"upperarm_twist_02_l\"),Weight=0.0287),(Item=(Type=Bone,Name=\"deform_upperarm_up_l\"),Weight=0.0288),(Item=(Type=Bone,Name=\"spine ...(truncated)"
    Weight=1
   )
   > &RigUnit_ParentConstraint(
    AdvancedSettings=(RotationOrderForFilter=XZY)
    Child=(Type=Bone,Name="dyn_buckle_r_MSTR")
    Parents="((Item=(Type=Bone,Name=\"head\"),Weight=0.0216),(Item=(Type=Bone,Name=\"deform_upperarm_fr_r\"),Weight=0.0251),(Item=(Type=Bone,Name=\"upperarm_twist_02_r\"),Weight=0.0285),(Item=(Type=Bone,Name=\"deform_upperarm_up_r\"),Weight=0.0286),(Item=(Type=Bone,Name=\"spine ...(truncated)"
    Weight=1
   )
   > &RigUnit_ParentConstraint(
    AdvancedSettings=(RotationOrderForFilter=XZY)
    Child=(Type=Bone,Name="dyn_box_r_small_MSTR")
    Parents="((Item=(Type=Bone,Name=\"deform_upperarm_bk_r\"),Weight=0.0149),(Item=(Type=Bone,Name=\"upperarm_r\"),Weight=0.0186),(Item=(Type=Bone,Name=\"upperarm_twist_02_r\"),Weight=0.0216),(Item=(Type=Bone,Name=\"deform_upperarm_up_r\"),Weight=0.0235),(Item=(Type=Bone,Name= ...(truncated)"
    Weight=1
   )
   > &RigUnit_ParentConstraint(
    AdvancedSettings=(RotationOrderForFilter=XZY)
    Child=(Type=Bone,Name="dyn_wire_MSTR")
    Parents="((Item=(Type=Bone,Name=\"deform_upperarm_up_r\"),Weight=0.0139),(Item=(Type=Bone,Name=\"deform_upperarm_bk_r\"),Weight=0.016),(Item=(Type=Bone,Name=\"upperarm_r\"),Weight=0.0166),(Item=(Type=Bone,Name=\"upperarm_twist_02_r\"),Weight=0.017),(Item=(Type=Bone,Name=\"c ...(truncated)"
    Weight=1
   )
   &RigUnit_ParentConstraint(
    AdvancedSettings=(InterpolationType=Average,RotationOrderForFilter=XZY)
    Child=(Type=Bone,Name="deform_head_rear_root")
    Parents=((Item=(Type=Bone,Name="head"),Weight=0.5),(Item=(Type=Bone,Name="neck_twist"),Weight=0.5))
    Weight=1
    ExecutePin=@RigVMFunction_Sequence.C
   )
   > &RigUnit_ParentConstraint(
    AdvancedSettings=(InterpolationType=Average,RotationOrderForFilter=XZY)
    Child=(Type=Bone,Name="deform_head_l_root")
    Parents=((Item=(Type=Bone,Name="head"),Weight=0.5),(Item=(Type=Bone,Name="neck_twist"),Weight=0.5))
    Weight=1
   )
   > &RigUnit_ParentConstraint(
    AdvancedSettings=(InterpolationType=Average,RotationOrderForFilter=XZY)
    Child=(Type=Bone,Name="deform_head_r_root")
    Parents=((Item=(Type=Bone,Name="head"),Weight=0.5),(Item=(Type=Bone,Name="neck_twist"),Weight=0.5))
    Weight=1
   )
   n1 = &RigUnit_SphericalPoseReader(
    ActiveRegionScaleFactors=(PositiveWidth=1,NegativeWidth=3,PositiveHeight=3.04,NegativeHeight=3)
    ActiveRegionSize=0.1
    Debug=(bDrawDebug=true,bDraw2D=False,bDrawLocalAxes=False,DebugScale=25,DebugSegments=20,DebugThickness=0.25)
    DriverAxis=(X=1,Y=0,Z=0)
    DriverItem=(Type=Bone,Name="head")
    FalloffRegionScaleFactors=(PositiveWidth=1,NegativeWidth=1,PositiveHeight=1,NegativeHeight=1)
    FalloffSize=0.3
    FlipHeightScaling=False
    FlipWidthScaling=False
    OptionalParentItem=(Type=Bone,Name="spine_05")
    RotationOffset=(X=-90,Y=0,Z=20)
    ExecutePin=@RigVMFunction_Sequence.E
   )
   > &RigUnit_ModifyTransforms(
    ItemToModify=((Item=(Type=Bone,Name="deform_head_rear"),Transform=(Rotation=(X=0,Y=0,Z=0.0872,W=0.9962),Translation=(X=3,Y=-10,Z=0),Scale3D=(X=1,Y=1,Z=1))))
    WeightMinimum=0
    Weight=@n1.OutputParam
   )
   > n2 = &RigUnit_SphericalPoseReader(
    ActiveRegionScaleFactors=(PositiveWidth=1,NegativeWidth=1,PositiveHeight=1,NegativeHeight=1)
    ActiveRegionSize=0.075
    Debug=(bDrawDebug=true,bDraw2D=False,bDrawLocalAxes=False,DebugScale=25,DebugSegments=20,DebugThickness=0.25)
    DriverAxis=(X=1,Y=0,Z=0)
    DriverItem=(Type=Bone,Name="head")
    FalloffRegionScaleFactors=(PositiveWidth=1,NegativeWidth=1,PositiveHeight=1,NegativeHeight=1)
    FalloffSize=0.25
    FlipHeightScaling=False
    FlipWidthScaling=False
    OptionalParentItem=(Type=Bone,Name="spine_05")
    RotationOffset=(X=0,Y=-45,Z=0)
   )
   > &RigUnit_ModifyTransforms(
    ItemToModify=((Item=(Type=Bone,Name="deform_head_r"),Transform=(Rotation=(X=0,Y=0,Z=0.0872,W=0.9962),Translation=(X=-5,Y=0,Z=7),Scale3D=(X=1.2,Y=1.2,Z=1.2))))
    WeightMinimum=0
    Weight=@n2.OutputParam
   )
   > n3 = &RigUnit_SphericalPoseReader(
    ActiveRegionScaleFactors=(PositiveWidth=1,NegativeWidth=1,PositiveHeight=1,NegativeHeight=1)
    ActiveRegionSize=0.075
    Debug=(bDrawDebug=true,bDraw2D=False,bDrawLocalAxes=False,DebugScale=25,DebugSegments=20,DebugThickness=0.25)
    DriverAxis=(X=1,Y=0,Z=0)
    DriverItem=(Type=Bone,Name="head")
    FalloffRegionScaleFactors=(PositiveWidth=1,NegativeWidth=1,PositiveHeight=1,NegativeHeight=1)
    FalloffSize=0.25
    FlipHeightScaling=False
    FlipWidthScaling=False
    OptionalParentItem=(Type=Bone,Name="spine_05")
    RotationOffset=(X=0,Y=225,Z=0)
   )
   > &RigUnit_ModifyTransforms(
    ItemToModify=((Item=(Type=Bone,Name="deform_head_l"),Transform=(Rotation=(X=0,Y=0,Z=0.0872,W=0.9962),Translation=(X=5,Y=0,Z=-7),Scale3D=(X=1.2,Y=1.2,Z=1.2))))
    WeightMinimum=0
    Weight=@n3.OutputParam
   )
   > n4 = &RigUnit_SphericalPoseReader(
    ActiveRegionScaleFactors=(PositiveWidth=1,NegativeWidth=1,PositiveHeight=1,NegativeHeight=1)
    ActiveRegionSize=0.1
    Debug=(bDrawDebug=true,bDraw2D=False,bDrawLocalAxes=False,DebugScale=25,DebugSegments=20,DebugThickness=0.25)
    DriverAxis=(X=0,Y=-1,Z=0)
    DriverItem=(Type=Bone,Name="head")
    FalloffRegionScaleFactors=(PositiveWidth=1,NegativeWidth=1,PositiveHeight=1,NegativeHeight=1)
    FalloffSize=0.3
    FlipHeightScaling=False
    FlipWidthScaling=False
    OptionalParentItem=(Type=Bone,Name="spine_05")
    RotationOffset=(X=20,Y=-10,Z=20)
   )
   > &RigUnit_ModifyTransforms(
    ItemToModify="((Item=(Type=Bone,Name=\"deform_head_rear\"),Transform=(Rotation=(X=0,Y=0,Z=0,W=1),Translation=(X=0,Y=-10,Z=0),Scale3D=(X=1,Y=1,Z=1))),(Item=(Type=Bone,Name=\"deform_head_r\"),Transform=(Rotation=(X=0,Y=0,Z=0,W=1),Translation=(X=0,Y=0,Z=0),Scale3D=(X=1,Y=1,Z=1 ...(truncated)"
    WeightMinimum=0
    Weight=@n4.OutputParam
   )
   > n5 = &RigUnit_SphericalPoseReader(
    ActiveRegionScaleFactors=(PositiveWidth=1,NegativeWidth=1,PositiveHeight=1,NegativeHeight=1)
    ActiveRegionSize=0.1
    Debug=(bDrawDebug=true,bDraw2D=False,bDrawLocalAxes=False,DebugScale=25,DebugSegments=20,DebugThickness=0.25)
    DriverAxis=(X=0,Y=-1,Z=0)
    DriverItem=(Type=Bone,Name="head")
    FalloffRegionScaleFactors=(PositiveWidth=1,NegativeWidth=1,PositiveHeight=1,NegativeHeight=1)
    FalloffSize=0.3
    FlipHeightScaling=false
    FlipWidthScaling=false
    OptionalParentItem=(Type=Bone,Name="spine_05")
    RotationOffset=(X=190,Y=-10,Z=20)
   )
   > &RigUnit_ModifyTransforms(
    ItemToModify="((Item=(Type=Bone,Name=\"deform_head_rear\"),Transform=(Rotation=(X=0,Y=0,Z=0,W=1),Translation=(X=0,Y=-10,Z=0),Scale3D=(X=1,Y=1,Z=1))),(Item=(Type=Bone,Name=\"deform_head_r\"),Transform=(Rotation=(X=0,Y=0,Z=0,W=1),Translation=(X=0,Y=0,Z=0),Scale3D=(X=1,Y=1,Z=1 ...(truncated)"
    WeightMinimum=0
    Weight=@n5.OutputParam
   )
   RigUnit_GetTransform = &RigUnit_GetTransform(Item=(Type=Bone,Name="head"), Space=LocalSpace, bInitial=False)
   RigVMFunction_MathQuaternionSwingTwist = &RigVMFunction_MathQuaternionSwingTwist(TwistAxis=(X=1,Y=0,Z=0), Input=@RigUnit_GetTransform.Transform.Rotation)
   Interpolate = /Script/RigVM.RigVMFunction_MathQuaternionSlerp(A=(X=0,Y=0,Z=0,W=1), T=0.5, B=@RigVMFunction_MathQuaternionSwingTwist.Twist) {is=struct RigVMUnitNode, template="Interpolate::Execute(in A,in B,in T,out Result)"}
   &RigUnit_ModifyTransforms(
    ItemToModify=((Item=(Type=Bone,Name="neck_twist"),Transform=(Rotation=(X=0,Y=0,Z=0,W=1),Translation=(X=0,Y=0,Z=0),Scale3D=(X=1,Y=1,Z=1))))
    Weight=1
    WeightMinimum=0
    ItemToModify.0.Transform.Rotation=@Interpolate.Result
    ExecutePin=@RigVMFunction_Sequence.D
   )
   n6 = &RigUnit_GetTransform(Item=(Type=Bone,Name="head"), Space=LocalSpace, bInitial=False)
   > &RigUnit_ModifyTransforms(
    ItemToModify=((Item=(Type=Bone,Name="neck_twist_lower"),Transform=(Rotation=(X=0,Y=0,Z=0,W=1),Translation=(X=0,Y=0,Z=0),Scale3D=(X=1,Y=1,Z=1))))
    Weight=1
    WeightMinimum=0
    ItemToModify.0.Transform.Rotation=&RigVMFunction_MathQuaternionSlerp(
     A=(X=0,Y=0,Z=0,W=1)
     T=0.2
     B=&RigVMFunction_MathQuaternionSwingTwist(TwistAxis=(X=1,Y=0,Z=0), Input=@n6.Transform.Rotation).Twist
    ) {template="Interpolate::Execute(in A,in B,in T,out Result)"}.Result
   )
   &RigUnit_ModifyTransforms(
    ExecutePin=()
    ItemToModify=((Item=(Type=Bone,Name="C_jaw_chin_deform"),Transform=(Rotation=(X=0.1874,Y=0,Z=0,W=0.9823),Translation=(X=0,Y=3.3337,Z=0.7017),Scale3D=(X=1,Y=0.875,Z=1))))
    Weight=1
    WeightMinimum=0
   )
   &RigUnit_SphericalPoseReader(
    ActiveRegionScaleFactors=(PositiveWidth=1,NegativeWidth=1,PositiveHeight=1,NegativeHeight=1)
    ActiveRegionSize=0.1
    Debug=(bDrawDebug=false,bDraw2D=False,bDrawLocalAxes=False,DebugScale=25,DebugSegments=20,DebugThickness=0.25)
    DriverAxis=(X=0,Y=1,Z=0)
    DriverItem=(Type=Bone,Name="C_jaw")
    ExecutePin=()
    FalloffRegionScaleFactors=(PositiveWidth=1,NegativeWidth=1,PositiveHeight=1,NegativeHeight=1)
    FalloffSize=0.2
    FlipHeightScaling=False
    FlipWidthScaling=False
    OptionalParentItem=(Type=Bone,Name="head")
    RotationOffset=(X=130,Y=0,Z=0)
   )
  ]}
]
hierarchy [
 root {loc=<0 0 0>, rot=<0 0 0 1>}
 pelvis {parent=root, loc=<0 -0.7321 92.4218>, rot=<0.0127 0.707 -0.0127 -0.707>}
 spine_01 {parent=pelvis, loc=<5.6002 -0.5128 0>, rot=<0 0 0.034 -0.9994>}
 spine_02 {parent=spine_01, loc=<8.0756 0.0139 0>, rot=<0 0 0.0238 -0.9997>}
 spine_03 {parent=spine_02, loc=<7.0395 0 0>, rot=<0 0 0.0149 -0.9999>}
 spine_04 {parent=spine_03, loc=<8.1023 0 0>, rot=<0 0 -0.0132 -0.9999>}
 spine_05 {parent=spine_04, loc=<10.0141 0 0>, rot=<0 0 -0.0015 -1>}
 clavicle_l {parent=spine_05, loc=<2.5856 -0.2866 -4.3777>, rot=<0.0465 -0.7289 -0.009 -0.6829>}
 upperarm_l {parent=clavicle_l, loc=<18.2264 0.1342 -0.001>, rot=<0.0253 -0.2693 0.0013 -0.9627>}
 lowerarm_l {parent=upperarm_l, loc=<29.0071 0.1686 -1.239>, rot=<0.0391 -0.0099 -0.2453 -0.9686>}
 hand_l {parent=lowerarm_l, loc=<25.8607 0.222 1.0274>, rot=<-0.7193 -0.0988 0.1193 -0.6773>}
 thumb_01_l {parent=hand_l, loc=<1.3118 -0.3191 -2.459>, rot=<0.6546 -0.4042 0.0664 -0.6354>}
 thumb_02_l {parent=thumb_01_l, loc=<5.9336 -0.0448 0.0401>, rot=<-0.0037 0.0213 0.298 -0.9543>}
 thumb_03_l {parent=thumb_02_l, loc=<2.2223 0.3682 -0.0907>, rot=<0.0118 -0.013 0.0066 -0.9998>}
 index_metacarpal_l {parent=hand_l, loc=<1.6535 0.0183 -2.96>, rot=<0.0158 -0.0124 -0.0173 -0.9996>}
 index_01_l {parent=index_metacarpal_l, loc=<8.3008 0.0357 -0.0524>, rot=<-0.0044 0.0485 0.0924 -0.9945>}
 index_02_l {parent=index_01_l, loc=<3.3587 -0.3004 0.0452>, rot=<-0.002 0.0027 0.3225 -0.9466>}
 index_03_l {parent=index_02_l, loc=<2.5578 -0.123 -0.0252>, rot=<-0.002 0.0067 0.2558 -0.9667>}
 ring_metacarpal_l {parent=hand_l, loc=<2.2074 -0.3625 1.3638>, rot=<-0.0284 0.0722 -0.0458 -0.9959>}
 ring_01_l {parent=ring_metacarpal_l, loc=<7.0295 -0.6197 -0.1446>, rot=<-0.0489 -0.0252 0.1892 -0.9804>}
 ring_02_l {parent=ring_01_l, loc=<3.6174 0.4647 -0.1027>, rot=<0 0 0.2806 -0.9598>}
 ring_03_l {parent=ring_02_l, loc=<2.4152 0.114 0.0199>, rot=<0 0 0.1189 -0.9929>}
 pinky_metacarpal_l {parent=hand_l, loc=<1.7616 -0.9753 3.3814>, rot=<-0.0397 0.0814 -0.069 -0.9935>}
 pinky_01_l {parent=pinky_metacarpal_l, loc=<7.1709 -0.4831 -0.1094>, rot=<-0.0654 -0.0035 0.2242 -0.9723>}
 pinky_02_l {parent=pinky_01_l, loc=<2.8274 0.0893 -0.1043>, rot=<0.0037 0.0011 0.2854 -0.9584>}
 pinky_03_l {parent=pinky_02_l, loc=<2.2046 0.0307 -0.0078>, rot=<-0.0002 0.0098 0.1445 -0.9895>}
 deform_wrist_dn_base_l {parent=hand_l, loc=<0 -3.4173 0>, rot=<0 0 0 -1>}
 deform_wrist_dn_l {parent=deform_wrist_dn_base_l, loc=<0 0 0>, rot=<0 0 0 1>}
 deform_wrist_up_base_l {parent=hand_l, loc=<0 2.7805 0>, rot=<0 0 0 -1>}
 deform_wrist_up_l {parent=deform_wrist_up_base_l, loc=<0 0 0>, rot=<0 0 0 1>}
 middle_metacarpal_l {parent=hand_l, loc=<2.1856 -0.1725 -0.5984>, rot=<0.006 0.0168 -0.0329 -0.9993>}
 middle_01_l {parent=middle_metacarpal_l, loc=<7.3275 -0.2447 0.0129>, rot=<-0.0208 0.0042 0.0858 -0.9961>}
 middle_02_l {parent=middle_01_l, loc=<4.0123 0.1401 0.1175>, rot=<-0.0218 0.0258 0.3512 -0.9357>}
 middle_03_l {parent=middle_02_l, loc=<2.6077 0.0534 0.1373>, rot=<0.0082 0.0188 0.1708 -0.9851>}
 lowerarm_twist_02_l {parent=lowerarm_l, loc=<9.6199 0 0>, rot=<0 0 0 -1>}
 lowerarm_twist_01_l {parent=lowerarm_l, loc=<17.8549 0 0>, rot=<0 0 0 -1>}
 deform_elbow_in_base_l {parent=lowerarm_l, loc=<0.2112 5.645 0.8274>, rot=<0 0 0 -1>}
 deform_elbow_in_l {parent=deform_elbow_in_base_l, loc=<0 0 0>, rot=<0 0 0 1>}
 deform_elbow_out_base_l {parent=lowerarm_l, loc=<0 -7.5129 0>, rot=<0 0 0 -1>}
 deform_elbow_out_l {parent=deform_elbow_out_base_l, loc=<0 0 0>, rot=<0 0 0 1>}
 upperarm_twist_01_l {parent=upperarm_l, loc=<7.9796 0.0843 -0.6195>, rot=<0 0 0 -1>}
 deform_upperarm_bk_base_l {parent=upperarm_twist_01_l, loc=<-6.8983 -13.3826 -1.3258>, rot=<0 0 0 1>}
 deform_upperarm_bk_l {parent=deform_upperarm_bk_base_l, loc=<0 0 0>, rot=<0 0 0 1>}
 deform_upperarm_fr_base_l {parent=upperarm_twist_01_l, loc=<-8.3487 10.7117 -2.1455>, rot=<0 0 0 1>}
 deform_upperarm_fr_l {parent=deform_upperarm_fr_base_l, loc=<0 0 0>, rot=<0 0 0 1>}
 deform_upperarm_up_base_l {parent=upperarm_twist_01_l, loc=<-11.2012 -1.2234 8.7784>, rot=<0 0 0 1>}
 deform_upperarm_up_l {parent=deform_upperarm_up_base_l, loc=<0 0 0>, rot=<0 0 0 1>}
 upperarm_twist_02_l {parent=upperarm_l, loc=<19.5059 0.0843 -0.6195>, rot=<0 0 0 -1>}
 deform_bicep_base_l {parent=upperarm_twist_02_l, loc=<0 8.0494 0>, rot=<0 0 0 1>}
 deform_bicep_l {parent=deform_bicep_base_l, loc=<0 0 0>, rot=<0 0 0 1>}
 neck_01 {parent=spine_05, loc=<10.815 1.9196 0>, rot=<0 0 -0.1864 -0.9825>}
 neck_02 {parent=neck_01, loc=<4.7347 0.3045 0>, rot=<0 0 0.0363 -0.9993>}
 head {parent=neck_02, loc=<5.7632 0.5552 0>, rot=<0 0 0.115 -0.9934>}
 faceAttach {parent=head, loc=<-4.146 -0.8545 0>, rot=<0.0029 -0.7071 -0.0029 -0.7071>}
 C_lip_upper_mid {parent=faceAttach, loc=<0.0022 16.8076 1.148>, rot=<0.2225 0.0013 -0.0005 -0.9749>}
 L_lip_upper_outer {parent=faceAttach, loc=<8.8441 14.829 1.3354>, rot=<0.0359 -0.0728 0.2627 -0.9615>}
 L_lip_corner {parent=faceAttach, loc=<14.176 12.1991 1.2088>, rot=<0.003 0.008 -0.0146 -0.9999>}
 R_lip_corner {parent=faceAttach, loc=<-14.176 12.1991 1.2088>, rot=<0.003 -0.008 0.0146 -0.9999>}
 C_jaw {parent=faceAttach, loc=<-0.0134 -0.1613 -1.4291>, rot=<0 0 0 1>}
 L_lip_lower_outer {parent=C_jaw, loc=<8.8405 14.8022 1.4143>, rot=<0.013 0.0445 0.3057 -0.951>}
 C_lip_lower_mid {parent=C_jaw, loc=<0.0223 16.7112 1.2233>, rot=<0 0 0 -1>}
 R_lip_lower_outer {parent=C_jaw, loc=<-8.8137 14.8022 1.4143>, rot=<0.013 -0.0445 -0.3057 -0.951>}
 teeth_lower {parent=C_jaw, loc=<0.0134 6.0835 -1.7637>, rot=<0 0 0 1>}
 tongue {parent=C_jaw, loc=<0.0134 -8.8951 -2.0419>, rot=<-0.1568 0 0 -0.9876>}
 teeth_upper {parent=faceAttach, loc=<0 6.5028 4.229>, rot=<0 0 0 1>}
 R_lip_upper_outer {parent=faceAttach, loc=<-8.8441 14.829 1.3354>, rot=<0.0359 0.0728 -0.2627 -0.9615>}
 dyn_hair_01 {parent=head, loc=<13.6636 6.6968 1.3731>, rot=<0.0992 -0.4448 -0.2073 -0.8656>}
 dyn_hair_02 {parent=dyn_hair_01, loc=<4.1335 0 0>, rot=<-0.0054 0.062 0.0283 -0.9977>}
 dyn_hair_03 {parent=dyn_hair_02, loc=<3.2479 0 0>, rot=<0.0011 0.037 0.0236 -0.999>}
 dyn_ear_l_01 {parent=head, loc=<11.412 -2.5996 -13.5489>, rot=<0.692 -0.1574 -0.0658 -0.7014>}
 dyn_ear_l_02 {parent=dyn_ear_l_01, loc=<4.9024 0 0>, rot=<-0.0073 -0.1525 0.0473 -0.9871>}
 dyn_ear_l_03 {parent=dyn_ear_l_02, loc=<4.6336 0 0>, rot=<0.0007 -0.1229 -0.006 -0.9924>}
 dyn_ear_r_01 {parent=head, loc=<11.412 -2.5996 13.5489>, rot=<0.7536 0.0532 0.1621 -0.6349>}
 dyn_ear_r_02 {parent=dyn_ear_r_01, loc=<4.9024 0 0>, rot=<0.004 -0.1434 -0.0701 -0.9872>}
 dyn_ear_r_03 {parent=dyn_ear_r_02, loc=<4.6336 0 0>, rot=<0.0023 -0.0907 -0.0256 -0.9955>}
 deform_head_rear_root {parent=head, loc=<-1.1977 -20.4696 0>, rot=<0 0 0 1>}
 deform_head_rear {parent=deform_head_rear_root, loc=<0 0 0>, rot=<0 0 0 1>}
 deform_head_l_root {parent=head, loc=<-1.5355 -4.591 -20.4092>, rot=<0.0021 -0.0099 -0.0008 -0.9999>}
 deform_head_l {parent=deform_head_l_root, loc=<0 0 0>, rot=<0 0 0 1>}
 deform_head_r_root {parent=head, loc=<-1.5353 -4.591 20.4092>, rot=<0 0 0 1>}
 deform_head_r {parent=deform_head_r_root, loc=<0 0 0>, rot=<0 0 0 1>}
 neck_twist {parent=neck_01, loc=<4.7347 0.3045 0>, rot=<0 0 0.0363 -0.9993>}
 clavicle_r {parent=spine_05, loc=<2.5855 -0.2866 4.3777>, rot=<0.7289 0.0465 -0.6829 0.009>}
 upperarm_r {parent=clavicle_r, loc=<-18.2264 -0.1342 0.001>, rot=<0.0253 -0.2693 0.0013 -0.9627>}
 lowerarm_r {parent=upperarm_r, loc=<-29.0071 -0.1686 1.239>, rot=<0.0391 -0.0099 -0.2453 -0.9686>}
 hand_r {parent=lowerarm_r, loc=<-25.8607 -0.222 -1.0274>, rot=<-0.7193 -0.0988 0.1193 -0.6773>}
 thumb_01_r {parent=hand_r, loc=<-1.3118 0.3191 2.459>, rot=<0.6546 -0.4042 0.0664 -0.6354>}
 thumb_02_r {parent=thumb_01_r, loc=<-5.9336 0.0448 -0.0401>, rot=<-0.0037 0.0213 0.298 -0.9543>}
 thumb_03_r {parent=thumb_02_r, loc=<-2.2223 -0.3682 0.0907>, rot=<0.0118 -0.013 0.0066 -0.9998>}
 index_metacarpal_r {parent=hand_r, loc=<-1.6535 -0.0183 2.96>, rot=<0.0158 -0.0124 -0.0173 -0.9996>}
 index_01_r {parent=index_metacarpal_r, loc=<-8.3008 -0.0357 0.0524>, rot=<-0.0044 0.0485 0.0924 -0.9945>}
 index_02_r {parent=index_01_r, loc=<-3.3587 0.3004 -0.0452>, rot=<-0.002 0.0027 0.3225 -0.9466>}
 index_03_r {parent=index_02_r, loc=<-2.5578 0.123 0.0252>, rot=<-0.002 0.0067 0.2558 -0.9667>}
 ring_metacarpal_r {parent=hand_r, loc=<-2.2074 0.3625 -1.3638>, rot=<-0.0284 0.0722 -0.0458 -0.9959>}
 ring_01_r {parent=ring_metacarpal_r, loc=<-7.0295 0.6197 0.1446>, rot=<-0.0489 -0.0252 0.1892 -0.9804>}
 ring_02_r {parent=ring_01_r, loc=<-3.6174 -0.4647 0.1027>, rot=<0 0 0.2806 -0.9598>}
 ring_03_r {parent=ring_02_r, loc=<-2.4152 -0.114 -0.0199>, rot=<0 0 0.1189 -0.9929>}
 pinky_metacarpal_r {parent=hand_r, loc=<-1.7616 0.9753 -3.3814>, rot=<-0.0397 0.0814 -0.069 -0.9935>}
 pinky_01_r {parent=pinky_metacarpal_r, loc=<-7.1709 0.4831 0.1094>, rot=<-0.0654 -0.0035 0.2242 -0.9723>}
 pinky_02_r {parent=pinky_01_r, loc=<-2.8274 -0.0893 0.1043>, rot=<0.0037 0.0011 0.2854 -0.9584>}
 pinky_03_r {parent=pinky_02_r, loc=<-2.2046 -0.0307 0.0078>, rot=<-0.0002 0.0098 0.1445 -0.9895>}
 deform_wrist_dn_base_r {parent=hand_r, loc=<0 3.4172 0>, rot=<0 0 0 -1>}
 deform_wrist_dn_r {parent=deform_wrist_dn_base_r, loc=<0 0 0>, rot=<0 0 0 -1>}
 deform_wrist_up_base_r {parent=hand_r, loc=<-0.0001 -2.7804 0>, rot=<0 0 0 -1>}
 deform_wrist_up_r {parent=deform_wrist_up_base_r, loc=<0 0 0>, rot=<0 0 0 -1>}
 middle_metacarpal_r {parent=hand_r, loc=<-2.1856 0.1725 0.5984>, rot=<0.006 0.0168 -0.0329 -0.9993>}
 middle_01_r {parent=middle_metacarpal_r, loc=<-7.3275 0.2447 -0.0129>, rot=<-0.0208 0.0042 0.0858 -0.9961>}
 middle_02_r {parent=middle_01_r, loc=<-4.0123 -0.1401 -0.1175>, rot=<-0.0218 0.0258 0.3512 -0.9357>}
 middle_03_r {parent=middle_02_r, loc=<-2.6077 -0.0534 -0.1373>, rot=<0.0082 0.0188 0.1708 -0.9851>}
 lowerarm_twist_02_r {parent=lowerarm_r, loc=<-9.6199 0 0>, rot=<0 0 0 -1>}
 lowerarm_twist_01_r {parent=lowerarm_r, loc=<-17.8549 0 0>, rot=<0 0 0 -1>}
 deform_elbow_in_base_r {parent=lowerarm_r, loc=<-0.2114 -5.6449 -0.827>, rot=<0 0 0 -1>}
 deform_elbow_in_r {parent=deform_elbow_in_base_r, loc=<0 0 0>, rot=<0 0 0 1>}
 deform_elbow_out_base_r {parent=lowerarm_r, loc=<-0.0001 7.5129 0.0002>, rot=<0 0 0 -1>}
 deform_elbow_out_r {parent=deform_elbow_out_base_r, loc=<0 0 0>, rot=<0 0 0 1>}
 upperarm_twist_01_r {parent=upperarm_r, loc=<-7.9796 -0.0843 0.6195>, rot=<0 0 0 -1>}
 deform_upperarm_bk_base_r {parent=upperarm_twist_01_r, loc=<6.8985 13.3826 1.3257>, rot=<0 0 0 1>}
 deform_upperarm_bk_r {parent=deform_upperarm_bk_base_r, loc=<0 0 0>, rot=<0 0 0 1>}
 deform_upperarm_fr_base_r {parent=upperarm_twist_01_r, loc=<8.3485 -10.7117 2.1456>, rot=<0 0 0 1>}
 deform_upperarm_fr_r {parent=deform_upperarm_fr_base_r, loc=<0 0 0>, rot=<0 0 0 1>}
 deform_upperarm_up_base_r {parent=upperarm_twist_01_r, loc=<11.2015 1.2234 -8.7788>, rot=<0 0 0 1>}
 deform_upperarm_up_r {parent=deform_upperarm_up_base_r, loc=<0 0 0>, rot=<0 0 0 1>}
 upperarm_twist_02_r {parent=upperarm_r, loc=<-19.5059 -0.0843 0.6195>, rot=<0 0 0 -1>}
 deform_bicep_base_r {parent=upperarm_twist_02_r, loc=<0.0002 -8.0494 -0.0003>, rot=<0 0 0 1>}
 deform_bicep_r {parent=deform_bicep_base_r, loc=<0 0 0>, rot=<0 0 0 1>}
 attach_cape {parent=spine_05, loc=<9.0993 -12.8051 0.1881>, rot=<0.0616 -0.7044 -0.0616 -0.7044>}
 cape {parent=attach_cape, type=Null, loc=<0 -9.6429 -5.8438>, rot=<0 0 0 1>}
 attach_backpack {parent=spine_05, loc=<-0.1067 -14.6399 0.2708>, rot=<-0.0093 -0.707 0.0093 -0.7071>}
 Backpack_BR {parent=attach_backpack, type=Null, loc=<0 -9.9112 -3.6552>, rot=<-0.0457 0 0 0.999>}
 dyn_box_r_lrg_MSTR {parent=spine_05, loc=<3.4291 -20.2654 19.7822>, rot=<0.6523 -0.3358 0.4634 -0.497>}
 dyn_box_r_lrg {parent=dyn_box_r_lrg_MSTR, loc=<0 0 0>, rot=<0 0 0 1>}
 dyn_wire_MSTR {parent=spine_05, loc=<0.2293 -19.7912 23.0672>, rot=<0.0722 -0.4864 0.8649 -0.1007>}
 dyn_wire_01 {parent=dyn_wire_MSTR, loc=<0 0 0>, rot=<0 0 0 1>}
 dyn_wire_02 {parent=dyn_wire_01, loc=<2.7016 0 0>, rot=<0.0003 -0.0499 0.0139 -0.9987>}
 dyn_wire_03 {parent=dyn_wire_02, loc=<3.0634 0 0>, rot=<0.0001 -0.0211 0.0056 -0.9998>}
 dyn_wire_04 {parent=dyn_wire_03, loc=<3.0975 0 0>, rot=<0.0003 -0.0391 0.0077 -0.9992>}
 dyn_box_r_small_MSTR {parent=spine_05, loc=<2.3762 -18.915 24.3679>, rot=<0.5758 -0.2997 0.5543 -0.521>}
 dyn_box_r_small {parent=dyn_box_r_small_MSTR, loc=<0 0 0>, rot=<0 0 0 1>}
 dyn_box_l_MSTR {parent=spine_05, loc=<4.3905 -16.9203 -25.5602>, rot=<0.5048 -0.4891 0.2546 -0.6642>}
 dyn_box_l {parent=dyn_box_l_MSTR, loc=<0 0 0>, rot=<0 0 0 1>}
 dyn_pouch_l_MSTR {parent=spine_05, loc=<-5.0587 16.2214 -28.8898>, rot=<-0.325 -0.6074 -0.381 -0.6166>}
 dyn_pouch_l {parent=dyn_pouch_l_MSTR, loc=<0 0 0>, rot=<0 0 0 1>}
 dyn_buckle_l_MSTR {parent=spine_05, loc=<6.1074 11.7145 -27.269>, rot=<-0.5184 -0.4688 -0.2065 -0.6847>}
 dyn_buckle_l {parent=dyn_buckle_l_MSTR, loc=<0 0 0>, rot=<0 0 0 1>}
 dyn_buckle_r_MSTR {parent=spine_05, loc=<6.1074 11.7145 27.269>, rot=<-0.6892 -0.2106 -0.467 -0.5124>}
 dyn_buckle_r {parent=dyn_buckle_r_MSTR, loc=<0 0 0>, rot=<0 0 0 1>}
 neck_twist_lower {parent=spine_05, loc=<10.815 1.9196 0>, rot=<0 0 -0.1506 -0.9886>}
 deform_pec_aim_base_l {parent=spine_04, loc=<5.0227 13.6719 -2.6508>, rot=<0.0672 -0.5201 0.3228 -0.7879>}
 deform_pec_aim_l {parent=deform_pec_aim_base_l, loc=<0 0 0>, rot=<0 0 0 1>}
 deform_pec_base_l {parent=deform_pec_aim_l, loc=<16.3279 0 0>, rot=<0 0 0 1>}
 deform_pec_l {parent=deform_pec_base_l, loc=<0 0 0>, rot=<0 0 0 1>}
 deform_pec_aim_base_r {parent=spine_04, loc=<5.0225 13.6718 2.6508>, rot=<-0.5201 -0.0672 0.7879 0.3228>}
 deform_pec_aim_r {parent=deform_pec_aim_base_r, loc=<0 0 0>, rot=<0 0 0 1>}
 deform_pec_base_r {parent=deform_pec_aim_r, loc=<-16.3279 0 0>, rot=<0 0 0 -1>}
 deform_pec_r {parent=deform_pec_base_r, loc=<0 0 0.0002>, rot=<0 0 0 -1>}
 deform_lat_l_base {parent=spine_03, loc=<2.5561 -10.3716 -17.6918>, rot=<0 0 0 -1>}
 deform_lat_l {parent=deform_lat_l_base, loc=<0 0 0>, rot=<0 0 0 1>}
 deform_lat_r_base {parent=spine_03, loc=<2.5562 -10.3716 17.6918>, rot=<0 0 1 0>}
 deform_lat_r {parent=deform_lat_r_base, loc=<0 0 0>, rot=<0 0 0 -1>}
 dyn_belly_mstr {parent=spine_01, loc=<0 9 0>, rot=<-0.4919 -0.508 -0.508 -0.4919>}
 dyn_belly {parent=dyn_belly_mstr, loc=<0 0 0>, rot=<0 0 0 1>}
 thigh_l {parent=pelvis, loc=<-1.8716 -0.0109 -11.9936>, rot=<0.0012 0.0572 0.0155 -0.9982>}
 calf_l {parent=thigh_l, loc=<-41.9178 -0.2246 0.0926>, rot=<0.0006 -0.0161 -0.0391 -0.9991>}
 foot_l {parent=calf_l, loc=<-38.7057 1.2438 0.2452>, rot=<0.0975 -0.0474 0.0317 -0.9936>}
 ball_l {parent=foot_l, loc=<-6.1614 13.1469 0.6807>, rot=<0.0001 -0.0001 -0.7063 -0.7079>}
 calf_twist_02_l {parent=calf_l, loc=<-15.9268 -0.0967 0.0397>, rot=<0.0237 0.0095 -0.0035 -0.9997>}
 calf_twist_01_l {parent=calf_l, loc=<-32.9335 -0.0967 0.0397>, rot=<0.0237 0.0095 -0.0035 -0.9997>}
 deform_knee_in_base_l {parent=calf_l, loc=<1.298 -8.0356 -0.0496>, rot=<0 0 0 -1>}
 deform_knee_in_l {parent=deform_knee_in_base_l, loc=<0 0 0>, rot=<0 0 0 1>}
 deform_knee_out_base_l {parent=calf_l, loc=<0 6.8739 0>, rot=<0 0 0 -1>}
 deform_knee_out_l {parent=deform_knee_out_base_l, loc=<0 0 0>, rot=<0 0 0 1>}
 thigh_twist_01_l {parent=thigh_l, loc=<-21.6303 0 0>, rot=<0.0013 0 0 -1>}
 deform_groin_base_l {parent=thigh_l, loc=<-2.1908 8.5661 -1.7442>, rot=<0 0 0 -1>}
 deform_groin_l {parent=deform_groin_base_l, loc=<0 0 0>, rot=<0 0 0 1>}
 thigh_r {parent=pelvis, loc=<-1.8716 -0.0109 11.9936>, rot=<0.0572 -0.0012 0.9982 0.0155>}
 calf_r {parent=thigh_r, loc=<41.9178 0.2246 -0.0926>, rot=<0.0006 -0.0161 -0.0391 -0.9991>}
 foot_r {parent=calf_r, loc=<38.7057 -1.2438 -0.2452>, rot=<0.0975 -0.0474 0.0317 -0.9936>}
 ball_r {parent=foot_r, loc=<6.1614 -13.1469 -0.6807>, rot=<0.0001 -0.0001 -0.7063 -0.7079>}
 calf_twist_02_r {parent=calf_r, loc=<15.9268 0.0967 -0.0397>, rot=<0.0237 0.0095 -0.0035 -0.9997>}
 calf_twist_01_r {parent=calf_r, loc=<32.9335 0.0967 -0.0397>, rot=<0.0237 0.0095 -0.0035 -0.9997>}
 deform_knee_in_base_r {parent=calf_r, loc=<-1.2981 8.0356 0.0496>, rot=<0 0 0 -1>}
 deform_knee_in_r {parent=deform_knee_in_base_r, loc=<0 0 0>, rot=<0 0 0 1>}
 deform_knee_out_base_r {parent=calf_r, loc=<0 -6.8739 0>, rot=<0 0 0 -1>}
 deform_knee_out_r {parent=deform_knee_out_base_r, loc=<0 0 0>, rot=<0 0 0 1>}
 thigh_twist_01_r {parent=thigh_r, loc=<21.6303 0 0>, rot=<0.0013 0 0 -1>}
 deform_groin_base_r {parent=thigh_r, loc=<2.1908 -8.5661 1.7442>, rot=<0 0 0 -1>}
 deform_groin_r {parent=deform_groin_base_r, loc=<0 0 0>, rot=<0 0 0 -1>}
 deform_glute_base_l {parent=pelvis, loc=<-10.1877 -11.6217 -12.899>, rot=<0.0275 0.0501 -0.4472 -0.8926>}
 deform_glute_l {parent=deform_glute_base_l, loc=<0 0 0>, rot=<0 0 0 1>}
 deform_glute_base_r {parent=pelvis, loc=<-10.1877 -11.6218 12.899>, rot=<0.0501 -0.0275 0.8926 -0.4472>}
 deform_glute_r {parent=deform_glute_base_r, loc=<0 0 0>, rot=<0 0 0 1>}
]
