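# Physics asset for the M_LRG_HarborSnow large male player cosmetic body, covering a dangling wire accessory rather than a full ragdoll.
# Four identical capsule links, dyn_wire_01 to 04, chain from a tiny dyn_wire_MSTR root body through constraints with matching tight swing and twist limits.
# Large unconstrained capsules on spine_05 and upperarm_r approximate the upper torso and right arm, apparently as colliders for the wire.
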PhysicsAsset /BRCosmetics/Characters/Player/Male/Large/Bodies/M_LRG_HarborSnow/Meshes/M_LRG_HarborSnow_Physics
bodies {
 dyn_wire_01 [{shapeType=Capsule, length=3, radius=4, shapeName=dyn_wire_01_capsule, center=<1 0 2.7>, rotation=<rot 90 0 0>}]
 dyn_wire_02 [{shapeType=Capsule, length=3, radius=4, shapeName=dyn_wire_02_capsule, center=<1 0 2.7>, rotation=<rot 90 0 0>}]
 dyn_wire_03 [{shapeType=Capsule, length=3, radius=4, shapeName=dyn_wire_03_capsule, center=<1 0 2.7>, rotation=<rot 90 0 0>}]
 dyn_wire_04 [{shapeType=Capsule, length=3, radius=4, shapeName=dyn_wire_04_capsule, center=<1 0 2.7>, rotation=<rot 90 0 0>}]
 dyn_wire_MSTR [{shapeType=Capsule, length=0.505, radius=0.505, shapeName=dyn_wire_MSTR_capsule}]
 spine_05 [{
   shapeType=Capsule
   length=37.8016
   radius=23.7862
   shapeName=spine_05_capsule
   center=<-9.0702 -5.6966 -8.5095>
   rotation=<rot 0 -0.0005 17.6326>
  }]
 upperarm_r [{shapeType=Capsule, length=24.9688, radius=15.6638, shapeName=upperarm_r_capsule, center=<-12.0217 0.8158 1.8895>, rotation=<rot 90 0 0>}]
}
constraints [
 {bone1Name=dyn_wire_01, bone2Name=dyn_wire_MSTR, swing1LimitDegrees=10, swing2LimitDegrees=20, twistLimitDegrees=4}
 {bone1Name=dyn_wire_02, bone2Name=dyn_wire_01, swing1LimitDegrees=10, swing2LimitDegrees=20, twistLimitDegrees=4}
 {bone1Name=dyn_wire_03, bone2Name=dyn_wire_02, swing1LimitDegrees=10, swing2LimitDegrees=20, twistLimitDegrees=4}
 {bone1Name=dyn_wire_04, bone2Name=dyn_wire_03, swing1LimitDegrees=10, swing2LimitDegrees=20, twistLimitDegrees=4}
]
