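# Post-process style AnimBlueprint for the large male HarborSnow body, copying its attached parent's pose through a Control Rig.
# An additive head bone rotation then counters a fraction of the head's offset from neck_02, only with a valid pawn owner and stronger during emotes.
# RigidBody physics based at spine_05 and a sprung AnimDynamics on dyn_belly add secondary motion last.
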AnimBlueprint /BRCosmetics/Characters/Player/Male/Large/Bodies/M_LRG_HarborSnow/Meshes/M_LRG_HarborSnow_AnimBP
target_skeleton=&/M_LRG_HarborSnow_Skeleton
graphs [
 AnimGraph {nodes [
   n1 = Control Rig(Source=Copy Pose From Mesh() {bUseAttachedParent=true}) {
    alphaBoolBlend {blendOption=Linear}
    controlRigAssetReference {blueprintRigClass=&/M_LRG_HarborSnow_CtrlRig.M_LRG_HarborSnow_CtrlRig_C}
   }
   n2 = "Transform (Modify) Bone - Bone: head"(
    ComponentPose=Local To Component(LocalPose=@n1)
    Alpha=Get HeadRotLimitAlpha() -> (HeadRotLimitAlpha:double)
    Translation="0,0,0"
    Rotation=Get HeadRot() -> (HeadRot:Rotator)
    Scale="1,1,1"
   ) {rotationMode=BMM_Additive, rotationSpace=BCS_BoneSpace, alphaBoolBlend {blendOption=Linear}, boneToModify {boneName=head}}
   n3 = RigidBody(ComponentPose=@n2) {
    bClampLinearTranslationLimitToRefPose=true
    overridePhysicsAsset=&/M_LRG_HarborSnow_Physics
    simulationSpace=BaseBoneSpace
    alphaBoolBlend {blendOption=Linear}
    baseBoneRef {boneName=spine_05}
    simSpaceSettings {
     dampingAlpha=0
     maxAngularAcceleration=40
     maxAngularVelocity=8
     maxLinearAcceleration=4000
     maxLinearVelocity=800
     worldAlpha=1
     externalLinearDragV=<0.1 0.1 0.2>
    }
   }
   n4 = Anim Dynamics - Bone: dyn_belly(ComponentPose=@n3, Alpha=1, ExternalForce="0,0,0") {
    angularDampingOverride=0.999
    angularSpringConstant=14000
    bAngularSpring=true
    bOverrideAngularDamping=true
    bUseGravityOverride=true
    simulationSpace=BoneRelative
    alphaBoolBlend {blendOption=Linear}
    boundBone {boneName=dyn_belly}
    componentAppliedLinearAccClamp=<10000 10000 10000>
    componentLinearAccScale=<0.6 0.6 0.6>
    componentLinearVelScale=<0.6 0.6 0.6>
    gravityOverride {z=-980}
    physicsBodyDefinitions [{
      boundBone {boneName=dyn_belly}
      boxExtents=<26 26 26>
      constraintSetup {angularLimitsMax=<10 20 30>, angularLimitsMin=<-10 -20 -30>, angularTarget=<1 0 0>}
      localJointOffset=<2 0 0>
     }]
    relativeSpaceBone {boneName=root}
    simSpaceSettings {maxAngularAcceleration=60, maxAngularVelocity=12, simSpaceAngularAlpha=0.2}
   }
   Output Pose(Result=Component To Local(ComponentPose=@n4)) {layerGroup=None}
  ]}
 EventGraph {nodes [
   Event BlueprintUpdateAnimation()
   > n1 = Sequence!(execute) -> (then_0, then_1, then_2:Exec)
   n2 = GetOwnerSkeletalMeshComponent()
   n3 = BreakRotator(InRot=BreakTransform(InTransform=GetDeltaTransformFromRefPose(self=@n2, BoneName=head, BaseName=neck_02)).Rotation)
   n4 = Get Z Percentage() -> (Z Percentage:double)
   n5 = Get X Percentage() -> (X Percentage:double)
   n6 = Get Y Percentage() -> (Y Percentage:double)
   > Set HeadRot(HeadRot:Rotator=MakeRotator(
     Roll=float * float!(A=@n3, B=-1, C:double=@n5) -> (ReturnValue) {is=title K2Node_PromotableOperator}
     Pitch=float * float(A=@n3.Pitch, B=-1, C=@n6)
     Yaw=float * float(A=@n3.Yaw, B=-1, C=@n4)
    )) -> (Output_Get:Rotator)
   n7 = Branch(execute=@n1.then_2, Condition=Get bIsPlayingEmote())
   > Set X Percentage(X Percentage:double=0.9) -> (Output_Get:double)
   > Set Y Percentage(Y Percentage:double=0.5) -> (Output_Get:double)
   > Set Z Percentage(Z Percentage:double=0.5) -> (Output_Get:double)
   Set X Percentage(execute=@n7.else, X Percentage=0.4)
   > Set Y Percentage(Y Percentage=0)
   > Set Z Percentage(Z Percentage=0)
   n8 = Is Valid(exec=@n1.then_1, InputObject=TryGetPawnOwner())
   Set HeadRotLimitAlpha(execute=@n8.Is Not Valid, HeadRotLimitAlpha:double=0) -> (Output_Get:double)
   Set HeadRotLimitAlpha(execute=@n8.Is Valid, HeadRotLimitAlpha=1)
   TryGetPawnOwner()
  ]}
]
variables [
 HeadRot {type=Rotator}
 HeadRotLimitAlpha {type=number}
 X Percentage {type=number, default=0.4}
 Y Percentage {type=number}
 Z Percentage {type=number, default=0.25}
]
