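# Physics asset of cape colliders for the large male HarborSnow player body, apparently the collision volumes that keep its cape from passing through the character.
# Capsules follow the pelvis, upper four spine bones (02 to 05), arms, hands, thighs, calves and feet, with two on the pelvis and none on head or neck.
# It has no constraints, and some left and right capsules differ in size, most visibly the hands.
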PhysicsAsset /BRCosmetics/Characters/Player/Male/Large/Bodies/M_LRG_HarborSnow/Meshes/M_LRG_HarborSnow_CapeColliders
bodies {
 calf_l [{shapeType=Capsule, length=37.0622, radius=16.9547, center=<-20.7639 -2.2427 0.8286>, rotation=<rot 90 0 0>}]
 calf_r [{shapeType=Capsule, length=37.0622, radius=16.9547, center=<20.7639 2.2427 -1.0553>, rotation=<rot 90 180 0>}]
 foot_l [{shapeType=Capsule, length=6.9942, radius=18.122, center=<-3.9164 3.986 0>, rotation=<rot 0 14.9998 81.2858>}]
 foot_r [{shapeType=Capsule, length=6.9931, radius=18.1232, center=<3.9164 -3.986 0>, rotation=<rot 0 14.9998 -94.9145>}]
 hand_l [{shapeType=Capsule, length=9.5015, radius=6.7933, center=<6.2703 0 0>, rotation=<rot 90 0 0>}]
 hand_r [{shapeType=Capsule, length=0.7694, radius=15.5254, center=<-6.2703 0 0>, rotation=<rot 90 0 180>}]
 lowerarm_l [{shapeType=Capsule, length=18.7132, radius=14.8257, center=<8.9852 0 -1.1217>, rotation=<rot 90 0 0>}]
 lowerarm_r [{shapeType=Capsule, length=19.6428, radius=16.6034, center=<-8.9852 0 0.3605>, rotation=<rot 90 180 0>}]
 pelvis [
  {shapeType=Capsule, length=19.3879, radius=26.6706, center=<-1.3986 1.111 0>}
  {shapeType=Capsule, length=15, radius=15, shapeName=pelvis_capsule, center=<-7.4928 0 -6.3344>}
 ]
 spine_02 [{shapeType=Capsule, length=16.8435, radius=27.9603, center=<-3.9557 1.6552 0>}]
 spine_03 [{shapeType=Capsule, length=0.36, radius=28.9847, center=<-1.2637 2.0534 0>}]
 spine_04 [{shapeType=Capsule, length=15.1298, radius=26.7499, center=<1.8749 1.3475 0>}]
 spine_05 [{shapeType=Capsule, length=36.4029, radius=22.805, center=<-6.7522 0.0604 0>}]
 thigh_l [{shapeType=Capsule, length=30.8965, radius=17.6186, center=<-22.1102 0 1.2756>, rotation=<rot 90 0 180>}]
 thigh_r [{shapeType=Capsule, length=26.6232, radius=19.1659, center=<22.1102 0 -1.2756>, rotation=<rot 90 180 180>}]
 upperarm_l [{shapeType=Capsule, length=17.8245, radius=14.9738, center=<10.0455 0 0>, rotation=<rot 90 10.7999 0>}]
 upperarm_r [{shapeType=Capsule, length=18.2422, radius=14.7979, center=<-10.0455 0 0>, rotation=<rot 90 -168.5981 0>}]
}
constraints []
